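# Additive unlit mesh-particle material for the GlowDonut Meltdown effect that draws animated pink-purple caustics with depth testing disabled.
# Two scrolling world-aligned water ripple normals form a NormalNoise offset that distorts the positions sampling two scrolling world-aligned caustic layers.
# The caustics blend dark and light color parameters into MF_FX_EmissiveAdjustments.
# Mobile shading path and Low quality use only the first caustic layer for the highlight blend.
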Material /WildEstate/TechArt/GlowDonut/Meltdown/Materials/M_GlowDonut_CausticProjection
graph {
 properties {
  bDisableDepthTest=true
  bUsedWithMeshParticles=true
  bUsedWithNiagaraMeshParticles=true
  blendMode=BLEND_Additive
  pixelDepthOffsetMode=PDOM_Legacy
  shadingModel=MSM_Unlit
  usedShadingModels=Unlit
 }
 nodes [
  n1 = VectorParameter() {parameterName=DarkColor, defaultValue=<rgba 0.0722 0.0161 0.0729 1>}
  n2 = VectorParameter() {parameterName=LightColor, defaultValue=<rgba 0.6 0.1145 0.5192 1>}
  n3 = Add(A=WorldPosition().XYZ, B=NamedRerouteUsage())
  n4 = VectorParameter() {parameterName=SpeedA, channelNames {a=Scalar, b="", g="", r=""}, defaultValue=<rgba 50 50 50 1>}
  n5 = Multiply(A=Time(), B=Multiply(A=@n4.RGB, B=@n4.RGB))
  n6 = ComponentMask(None=MaterialFunctionCall(
    TextureObject=TextureObject() {texture=&T_ShieldCaustics}
    TextureSize=Constant() {r=340}
    WorldPosition=Add(A=@n3, B=@n5)
   ) {materialFunction=&WorldAlignedTexture}.XY Texture) {r=true}
  n7 = TextureObject() {texture=&T_ShieldCaustics}
  n8 = Constant() {r=256}
  n9 = VectorParameter() {parameterName=SpeedB, channelNames {a=Scalar, b="", g="", r=""}, defaultValue=<rgba 50 50 50 -1>}
  n10 = Add(A=Add(A=WorldPosition().XYZ, B=NamedRerouteUsage()), B=Multiply(A=Time(), B=Multiply(A=@n9.RGB, B=@n9.RGB)))
  n11 = AppendVector(
   A=@n6
   B=ComponentMask(None=MaterialFunctionCall(TextureObject=@n7, TextureSize=@n8, WorldPosition=@n10) {materialFunction=&WorldAlignedTexture}.XY Texture) {r=true}
  )
  n12 = ShadingPathSwitch(Default=Desaturation(None=SquareRoot(None=@n11), Fraction=Constant() {r=0.7}), Mobile=@n6) {desc=Migrated from FeatureLevelSwitch}
  MaterialFunctionCall(Emissive=LinearInterpolate(A=LinearInterpolate(A=@n1.RGB, B=@n2.RGB, Alpha=@n6), B=@n2.RGB, Alpha=QualitySwitch(Default=@n12, Low=@n6))) {materialFunction=&MF_FX_EmissiveAdjustments}
  n13 = Time()
  n14 = Constant4Vector() {constant=<rgba 0.2 0 0 500>}
  n15 = Multiply(A=@n14.A, B=@n14.A)
  n16 = MaterialFunctionCall(
   TextureObject=TextureObject() {samplerType=SAMPLERTYPE_Normal, texture=/Game/Effects/Fort_Effects/Gameplay/Water/Textures/T_WaterRipples_01}
   TextureSize=Constant() {r=400}
   WorldPosition=Add(A=WorldPosition().XYZ, B=Multiply(A=@n13, B=@n15))
  ) {materialFunction=&WorldAlignedNormal}
  n17 = WorldPosition()
  n18 = Constant4Vector() {constant=<rgba -0.2 0 0 500>}
  n19 = Multiply(A=Time(), B=Multiply(A=@n18.A, B=@n18.A))
  n20 = Divide(A=MaterialFunctionCall(
    TextureObject=TextureObject() {samplerType=SAMPLERTYPE_Normal, texture=&T_WaterRipples_01}
    TextureSize=Constant() {r=200}
    WorldPosition=Add(A=@n17.XYZ, B=@n19)
   ) {materialFunction=&WorldAlignedNormal}.XY Texture)
  NamedRerouteDeclaration(None=Multiply(A=Add(A=@n16.XY Texture, B=@n20)) {constB=50}) {name=NormalNoise, nodeColor=<rgba 0 1 0.8359 1>}
 ]
}
