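# Masked unlit material for Glow Donut Meltdown particles, built on the shared NoTickParticles material functions.
# The base function supplies per-particle data and offsets, with radial force, vortex noise and curl noise chained onto world position offset.
# Billboard size combines sine pulsing, camera-distance shrink/grow, bounding-box and lifetime scaling, plus a camera offset.
# A soft circular mask gated by vertex red shapes particles, with rotatable UVs and eye-adaptation-compensated color.
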Material /WildEstate/TechArt/GlowDonut/Meltdown/Materials/M_GlowDonut_Meltdown_NoTickParticles_P
graph {
 properties {bAutomaticallySetUsageInEditor=false, bHasPixelAnimation=true, blendMode=BLEND_Masked, shadingModel=MSM_Unlit, usedShadingModels=Unlit}
 nodes [
  n1 = LinearInterpolate(
   A=ScalarParameter() {parameterName=BBX.CameraOffsetMin}
   B=ScalarParameter() {parameterName=BBX.CameraOffsetMax}
   Alpha=MaterialFunctionCall!(PartcileSpawnPos=NamedRerouteUsage()) {materialFunction=/WildEstate/TechArt/MaterialFunctions/NoTickParticles/MF_NoTickParticles_BoundingBoxScaling}.Mask without scale values
  )
  Add(
   A=Add(
    A=NamedRerouteUsage()
    B=MaterialFunctionCall(Offset Amount=@n1) {materialFunction=/Engine/Functions/Engine_MaterialFunctions02/WorldPositionOffset/CameraOffset}.Clamped Camera Offset
   )
   B=NamedRerouteUsage()
  )
  Divide(A=NamedRerouteUsage(), B=EyeAdaptation().EyeAdaptation)
  n2 = Constant()
  Multiply(
   A=If(A=VertexColor().R, "A > B"=@n2, "A == B"=@n2, "A < B"=Constant() {r=1}) {constB=1, equalsThreshold=0.001}
   B=SphereMask(A=TextureCoordinate(), B=Constant() {r=0.5}) {attenuationRadius=0.5, hardnessPercent=0.5}
  )
  n3 = ScalarParameter() {group=Forces, parameterName=WindSwayMinVal, sortPriority=40}
  n4 = Add(A=Time(), B=NamedRerouteUsage())
  n5 = ScalarParameter() {defaultValue=0.05, group=Forces, parameterName=WindSwaySpeed, sortPriority=40}
  n6 = ConstantBiasScale(None=Sine(None=Multiply(A=@n4, B=@n5)))
  n7 = MaterialFunctionCall!(
   UVs
   BypassBaseBillboardFunc=StaticBool() {value=true}
   UseBillboardWPO
   UseCurveForColorOpacity=StaticBoolParameter() {parameterName=UseCurveForOpacity}
   BoundModifier
   WindModifierMult=StaticSwitchParameter(True=LinearInterpolate(A=@n3, Alpha=@n6), False=Constant() {r=1}) {group=Forces, parameterName=UseWindModifier}
   UseDepthFadeInOpacity=StaticBool()
  ) {materialFunction=/WildEstate/TechArt/MaterialFunctions/NoTickParticles/MF_NoTickParticleBase}
  NamedRerouteDeclaration(None=@n7.UVs) {name=Base_UVs, nodeColor=<rgba 1 0 0.0234 1>}
  n8 = ScalarParameter() {defaultValue=0.1, group=scale, parameterName=PulseScaleRange, sortPriority=60}
  n9 = ComponentMask(None=NamedRerouteUsage()) {g=true}
  n10 = ScalarParameter() {defaultValue=10, group=scale, parameterName=PulseTimeSpread, sortPriority=60}
  n11 = Multiply(A=Add(A=Time(), B=Multiply(A=@n9, B=@n10)), B=ScalarParameter() {defaultValue=0.1, group=scale, parameterName=PulseSpeed, sortPriority=60}) {constB=0.5}
  n12 = LinearInterpolate(A=Subtract(B=@n8), B=Add(A=@n8), Alpha=ConstantBiasScale(None=Sine(None=@n11))) {constA=0.9, constB=1.1}
  n13 = MaterialFunctionCall!(Lifetime=NamedRerouteUsage(), PerParticleRandom=NamedRerouteUsage()) {materialFunction=/WildEstate/TechArt/MaterialFunctions/NoTickParticles/MF_NoTickParticle_ParticleScale}
  n14 = Add(A=NamedRerouteUsage(), B=ScalarParameter() {defaultValue=-500, group=Shrink, parameterName=Shrink.Offset})
  n15 = ScalarParameter() {defaultValue=3500, group=Shrink, parameterName=Shrink.NearStartDistance}
  n16 = StaticSwitch(
   True=Saturate(None=Divide(A=@n14, B=@n15))
   False=Constant() {r=1}
   Value=StaticBoolParameter() {group=Shrink, parameterName=Shrink.UseDistanceShrink, sortPriority=31}
  )
  n17 = NamedRerouteUsage()
  n18 = ScalarParameter() {defaultValue=5000, group=Shrink, parameterName=Grow.StartDistance}
  n19 = OneMinus(None=Divide(A=Subtract(A=@n17, B=@n18), B=ScalarParameter() {defaultValue=2000, parameterName=Grow.Range}))
  n20 = LinearInterpolate(A=ScalarParameter() {defaultValue=1, parameterName=Grow.Multi}, Alpha=SmoothStep(Value=Saturate(None=@n19))) {constA=1}
  n21 = Multiply(
   A=Multiply(A=Multiply(A=@n13."Scale (V2)", B=@n12), B=Multiply(A=@n16, B=@n20))
   B=MaterialFunctionCall(PartcileSpawnPos=NamedRerouteUsage()) {materialFunction=&MF_NoTickParticles_BoundingBoxScaling}.Scale Multiplier
  )
  n22 = NamedRerouteUsage()
  n23 = ScalarParameter() {defaultValue=0.9, group=Opacity, parameterName=LifetimeScaleFadeStart}
  n24 = Constant() {r=1}
  n25 = LinearInterpolate(
   A=ScalarParameter() {defaultValue=0.3, group=Opacity, parameterName=LifetimeScaleMin}
   Alpha=OneMinus(None=MaterialFunctionCall(Input=@n22, Input Low=@n23, Input High=@n24) {materialFunction=&RemapValue_0-1}.Result)
  )
  NamedRerouteDeclaration(None=MaterialFunctionCall!(UVs=NamedRerouteUsage(), Scale=Multiply(A=@n21, B=@n25), UV Center) {materialFunction=/WildEstate/TechArt/MaterialFunctions/MF_Billboard}.Position) {name=BillboardWPO, nodeColor=<rgba 0 0.0625 1 1>}
  NamedRerouteDeclaration(None=@n7.Color) {name=Base_Color, nodeColor=<rgba 1 0 0.0573 1>}
  NamedRerouteDeclaration(None=@n7.ParticleSpawnPosition) {name=Base_ParticleSpawnPos, nodeColor=<rgba 1 0 0.0573 1>}
  NamedRerouteDeclaration(None=@n7.Time) {name=Base_Time, nodeColor=<rgba 1 0 0.0573 1>}
  NamedRerouteDeclaration(None=@n7.WPO) {name=Base_WPO, nodeColor=<rgba 1 0 0.0573 1>}
  NamedRerouteDeclaration(None=VectorParameter() {group=Color, parameterName=Color, defaultValue=<rgba 0.3953 0.5573 0.4856 1>}.RGB) {name=Color, nodeColor=<rgba 0.9453 0 1 1>}
  n26 = NamedRerouteUsage()
  n27 = MaterialFunctionCall(Color=Subtract(A=Multiply(A=@n26) {constB=2})) {materialFunction=&SplitComponents}
  n28 = ScalarParameter() {
   defaultValue=1
   desc="Favor rotation one direction over the other. Or set to \"0\" for completely random"
   group=Rotation
   parameterName=RotationDirectionBias
   sliderMax=1
   sliderMin=-1
  }
  n29 = LinearInterpolate(B=@n27.G, Alpha=ScalarParameter() {group=Rotation, parameterName=RotationOffsetVariation, sliderMax=1})
  n30 = Multiply(A=NamedRerouteUsage(), B=ScalarParameter() {desc=Degrees Per Second, group=Rotation, parameterName=RotationSpeed})
  n31 = ScalarParameter() {group=Rotation, parameterName=RotationAngle}
  n32 = Multiply(A=Sign(None=Add(A=@n27.R, B=@n28)), B=Add(A=Multiply(A=@n29) {constB=180}, B=Add(A=@n30, B=@n31)))
  n33 = Divide() {constB=360}
  NamedRerouteDeclaration(None=MaterialFunctionCall(UVs=NamedRerouteUsage(), "Rotation Angle (0-1)"=Multiply(A=@n32, B=@n33)) {materialFunction=&CustomRotator}.Rotated Values) {name=RotatedUVs, nodeColor=<rgba 0.5703 1 0 1>}
  n34 = NamedRerouteUsage()
  n35 = ObjectPositionWS()
  NamedRerouteDeclaration(None=Distance(A=Add(A=@n34, B=@n35), B=CameraPositionWS())) {name=ParticleDistanceFromCamera, nodeColor=<rgba 0.3828 1 0 1>}
  NamedRerouteDeclaration(None=@n7.ActorRandomSeed) {name=Base_Seed, nodeColor=<rgba 0 0.625 1 1>}
  NamedRerouteDeclaration(None=@n7.Lifetime) {name=Base_ParticleLifetime, nodeColor=<rgba 1 0 0.0573 1>}
  n36 = MaterialFunctionCall!(
   WPO In=NamedRerouteUsage()
   Lifetime=NamedRerouteUsage()
   ParticlePosition=NamedRerouteUsage()
   Radial Force Center=VectorParameter() {group=Radial Force, parameterName=Radial Force Center, defaultValue=<rgba 0 0 0 1>}.RGB
  ) {materialFunction=/WildEstate/TechArt/MaterialFunctions/NoTickParticles/MF_NoTickParticle_RadialForce}
  n37 = NamedRerouteUsage()
  n38 = NamedRerouteUsage()
  n39 = NamedRerouteUsage()
  n40 = NamedRerouteUsage()
  NamedRerouteDeclaration(None=MaterialFunctionCall!(
    WPO=MaterialFunctionCall!(WPO=@n36.WPO, Lifetime=@n37, ParticleRandColor=@n38, ParticleSpawnPos=@n39, Time=@n40) {materialFunction=/WildEstate/TechArt/MaterialFunctions/NoTickParticles/MF_NoTickParticle_VortexNoise}.WPO
    Lifetime=NamedRerouteUsage()
    ParticleRandomColor=NamedRerouteUsage()
    Time=NamedRerouteUsage()
   ) {materialFunction=/WildEstate/TechArt/MaterialFunctions/NoTickParticles/MF_NoTickParticle_CurlNoise}.WPO) {name=ProcessedWPO_ParticleOffsets, nodeColor=<rgba 1 0 0.8906 1>}
  NamedRerouteDeclaration(None=@n7.Opacity) {name=Base_ParticleOpacity, nodeColor=<rgba 1 0 0.0573 1>}
  NamedRerouteDeclaration(None=@n7.PerParticleRandom) {name=Base_ParticleRandomColor, nodeColor=<rgba 1 0 0.0573 1>}
 ]
}
