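# MF_Meltdown_Utils is a shared material function for the GlowDonut Meltdown materials, providing a colour palette and an offset time value.
# It exposes Primary, Secondary and Tertiary colour vector parameters, defaulting to pink and purple, as RGB outputs through named reroutes.
# RandomTime adds Time, the actor's world-position length modulo a constant, and summed vertex colour components, desynchronising animation across actors and painted vertices.
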MaterialFunction /WildEstate/TechArt/GlowDonut/Meltdown/Functions/MF_Meltdown_Utils
graph {nodes [
  FunctionOutput(None=NamedRerouteDeclaration(None=VectorParameter() {parameterName=Color.Primary, defaultValue=<rgba 1 0.0723 1 1>}.RGB) {name=Color.Prim, nodeColor=<rgba 0 1 0.9062 1>}) {outputName=PrimaryColor, sortPriority=1}
  FunctionOutput(None=NamedRerouteDeclaration(None=VectorParameter() {parameterName=Color.Secondary, defaultValue=<rgba 0.7084 0 0.9823 1>}.RGB) {name=Color.Sec, nodeColor=<rgba 1 0.5859 0 1>}) {outputName=SecondaryColor, sortPriority=2}
  FunctionOutput(None=NamedRerouteDeclaration(None=VectorParameter() {parameterName=Color.Tertiary, defaultValue=<rgba 0.3381 0 0.4688 1>}.RGB) {name=Color.Tertiary, nodeColor=<rgba 1 0 0.5541 1>}) {outputName=TertiaryColor, sortPriority=3}
  n1 = Time()
  n2 = Modulo!(A=Length(None=ActorPositionWS()) {is=class MaterialExpressionLength}, B=Constant() {r=999})
  n3 = VertexColor()
  FunctionOutput(None=Add(A=Add(A=@n1, B=@n2), B=MaterialFunctionCall(f3=@n3) {materialFunction=&AddComponents}.f3)) {outputName=RandomTime, sortPriority=4}
 ]}
