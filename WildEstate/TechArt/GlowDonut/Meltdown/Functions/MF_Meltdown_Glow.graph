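# Material function for the GlowDonut Meltdown glow, outputting material attributes with a flickering emissive color and a soft Fresnel opacity.
# Flicker combines sines of the RandomTime input at offset frequencies (R² × B) to blend a brightness multiplier onto the Color input.
# Opacity is MF_Fresnel_FX scaled and softened by DepthFade, then multiplied by an object-to-camera distance fade that hides the glow up close.
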MaterialFunction /WildEstate/TechArt/GlowDonut/Meltdown/Functions/MF_Meltdown_Glow
graph {nodes [
  n1 = Time()
  n2 = Add(
   A=Multiply(A=FunctionInput(Preview=@n1) {inputName=RandomTime, inputType=FunctionInput_Scalar}, B=Constant3Vector() {constant=<rgba 1.13 2.87 1.62 0>})
   B=Constant3Vector() {constant=<rgba 0.26 0.56 0.867 0>}
  )
  n3 = MaterialFunctionCall(Float3=ConstantBiasScale(None=Sine(None=@n2))) {materialFunction=&BreakOutFloat3Components}
  n4 = LinearInterpolate(
   B=ScalarParameter() {defaultValue=1.5, group=Glow, parameterName=Glow.EmissiveFlicker.Mult}
   Alpha=Multiply(A=Multiply(A=@n3.R, B=@n3.R), B=@n3.B)
  ) {constA=1, constB=1.2}
  n5 = FunctionInput() {inputName=Color}
  n6 = ScalarParameter() {defaultValue=1, group=Glow, parameterName=Glow.Fresnel.Brightness}
  n7 = ScalarParameter() {defaultValue=3, group=Glow, parameterName=Glow.Fresnel.Exponent}
  n8 = DepthFade(
   Opacity=Multiply(
    A=MaterialFunctionCall(Brightness=@n6, Exponent=@n7) {materialFunction=&MF_Fresnel_FX}.Fresnel
    B=ScalarParameter() {defaultValue=0.25, group=Glow, parameterName=Glow.Opacity.Multi}
   )
   FadeDistance=ScalarParameter() {defaultValue=1500, group=Glow, parameterName=Glow.Opacity.DepthFade}
  )
  n9 = QualitySwitch(Default=@n8, High=@n8, Epic=@n8)
  n10 = Distance(A=ObjectPositionWS(), B=CameraPositionWS())
  n11 = ScalarParameter() {defaultValue=-1500, group=Glow, parameterName=Glow.Opacity.DistanceFadeOffset}
  n12 = Saturate(None=Divide(A=Add(A=@n10, B=@n11), B=ScalarParameter() {defaultValue=800, group=Glow, parameterName=Glow.Opacity.DistanceFade}))
  FunctionOutput(None=SetMaterialAttributes!(MaterialAttributes, EmissiveColor=Multiply(A=@n4, B=@n5), Opacity=Multiply(A=@n9, B=@n12)) {attributeSetTypes [#1, #2]}) {outputName=Material}
  WorldPosition()
 ]}
