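# M_Meltdown_Glow_P is an unlit, additive, two-sided glow material for the GlowDonut Meltdown effect set.
# It is a thin wrapper that uses material attributes and leaves the shading logic to two shared material functions.
# MF_Meltdown_Utils supplies its SecondaryColor output to both the Color and RandomTime inputs of MF_Meltdown_Glow.
# Automatic usage flagging in the editor is disabled.
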Material /WildEstate/TechArt/GlowDonut/Meltdown/Materials/M_Meltdown_Glow_P
graph {
 properties {
  bAutomaticallySetUsageInEditor=false
  bUseMaterialAttributes=true
  blendMode=BLEND_Additive
  shadingModel=MSM_Unlit
  twoSided=true
  usedShadingModels=Unlit
 }
 nodes [
  n1 = MaterialFunctionCall() {materialFunction=/WildEstate/TechArt/GlowDonut/Meltdown/Functions/MF_Meltdown_Utils}
  MaterialFunctionCall!(Color=@n1.SecondaryColor, RandomTime=@n1.SecondaryColor) {materialFunction=/WildEstate/TechArt/GlowDonut/Meltdown/Functions/MF_Meltdown_Glow}
 ]
}
