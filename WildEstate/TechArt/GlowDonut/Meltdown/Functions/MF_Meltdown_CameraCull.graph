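# Material function for the Meltdown effect that culls opacity near the camera, within an adjustable distance broken up by animated noise.
# A world-aligned Perlin texture scrolls over time and offsets world position before the camera distance is measured.
# The distance plus a negative threshold parameter is saturated and multiplied into the Opacity input.
# A divide that would widen the fade is left unconnected, so the cutoff is nearly hard.
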MaterialFunction /WildEstate/TechArt/GlowDonut/Meltdown/Functions/MF_Meltdown_CameraCull
graph {nodes [
  n1 = WorldPosition()
  n2 = WorldPosition()
  n3 = Multiply(A=Time(), B=VectorParameter() {group=Meltdown, parameterName=CameraCull.NoiseSpeed, defaultValue=<rgba 100 100 -400 1>}.RGB) {constB=100}
  n4 = Multiply(A=MaterialFunctionCall(
    TextureObject=TextureObject() {samplerType=SAMPLERTYPE_LinearColor, texture=/Game/Global/Textures/Default/Tileable_Patterns/T_PerlinNoise_Uncompressed}
    TextureSize=Constant() {r=1000}
    WorldPosition=Add(A=@n2.XYZ, B=@n3)
   ) {materialFunction=&WorldAlignedTexture}.XYZ Texture) {constB=500}
  n5 = Add(
   A=Distance(A=Add(A=@n1.XYZ, B=@n4), B=CameraPositionWS())
   B=ScalarParameter() {defaultValue=-2500, group=Meltdown, parameterName=CameraDistanceCull.Distance}
  )
  Divide(A=@n5) {constB=1000}
  n6 = FunctionInput() {bUsePreviewValueAsDefault=true, inputName=Opacity, inputType=FunctionInput_Scalar, previewValue=<1 0 1 1>}
  FunctionOutput(None=Multiply(A=@n6, B=Saturate(None=@n5)))
 ]}
