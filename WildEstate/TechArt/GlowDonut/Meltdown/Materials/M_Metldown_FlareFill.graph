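# Unlit, two-sided masked material for the Meltdown flare fill, drawing a radial colour gradient with a thin outline band.
# Two panning Perlin noise samples warp the gradient UVs, with colours supplied by MF_Meltdown_Utils.
# MF_Meltdown_CameraCull drives the opacity mask outside shadow passes, and World Position Offset lowers the mesh when the camera is below it.
# Emissive is partly eye-adaptation compensated, and a DebugOff static switch blanks the output.
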Material /WildEstate/TechArt/GlowDonut/Meltdown/Materials/M_Metldown_FlareFill
graph {
 properties {
  bAutomaticallySetUsageInEditor=false
  bHasPixelAnimation=true
  bUseMaterialAttributes=true
  blendMode=BLEND_Masked
  shadingModel=MSM_Unlit
  twoSided=true
  usedShadingModels=Unlit
 }
 nodes [
  n1 = MaterialFunctionCall() {materialFunction=/WildEstate/TechArt/GlowDonut/Meltdown/Functions/MF_Meltdown_Utils}
  NamedRerouteDeclaration(None=@n1.PrimaryColor) {name=Color.Prim, nodeColor=<rgba 0 1 0.9063 1>}
  NamedRerouteDeclaration(None=TextureObject() {samplerType=SAMPLERTYPE_LinearColor, texture=/Game/Global/Textures/Default/Tileable_Patterns/T_PerlinNoise_Uncompressed}) {name=NoiseTextureObject, nodeColor=<rgba 0 0.4844 1 1>}
  NamedRerouteDeclaration(None=@n1.SecondaryColor) {name=Color.Sec, nodeColor=<rgba 1 0.5859 0 1>}
  NamedRerouteDeclaration(None=@n1.TertiaryColor) {name=Color.Tertiary, nodeColor=<rgba 1 0 0.5541 1>}
  NamedRerouteDeclaration(None=@n1.RandomTime) {name=RandomTime, nodeColor=<rgba 1 0 0.1641 1>}
  n2 = Multiply(A=Panner(Time=NamedRerouteUsage()) {speedX=0.02, speedY=0.04})
  n3 = NamedRerouteUsage()
  n4 = Multiply(A=Panner() {speedX=-0.05, speedY=-0.1})
  n5 = Multiply(A=AppendVector(
    A=TextureSample(UVs=@n2, Tex=@n3) {samplerType=SAMPLERTYPE_LinearColor, texture=&T_PerlinNoise_Uncompressed}.R
    B=TextureSample(UVs=@n4) {samplerType=SAMPLERTYPE_LinearColor, texture=&T_PerlinNoise_Uncompressed}.R
   )) {constB=2}
  NamedRerouteDeclaration(None=Multiply(A=Subtract(A=@n5), B=ScalarParameter() {defaultValue=0.03, group=Noise, parameterName=WarpAmount}) {constB=0.03}) {name=UVWarp, nodeColor=<rgba 0.2656 1 0 1>}
  n6 = Constant()
  n7 = Add(A=TextureCoordinate(), B=NamedRerouteUsage())
  n8 = ScalarParameter() {defaultValue=0.4, parameterName=CenterColorRadius}
  n9 = Constant() {r=100}
  n10 = MaterialFunctionCall(UVs=@n7, Radius=@n8, Density=@n9) {materialFunction=&RadialGradientExponential}
  n11 = LinearInterpolate(A=NamedRerouteUsage(), B=NamedRerouteUsage(), Alpha=@n10.RadialGradientExponential)
  n12 = NamedRerouteUsage()
  n13 = ScalarParameter() {defaultValue=0.01, parameterName=LineThickness}
  n14 = Subtract(
   A=MaterialFunctionCall(UVs=@n7, Radius=Add(A=@n8, B=@n13), Density=@n9) {materialFunction=&RadialGradientExponential}.RadialGradientExponential
   B=@n10.RadialGradientExponential
  )
  n15 = MaterialFunctionCall!(Opacity) {materialFunction=/WildEstate/TechArt/GlowDonut/Meltdown/Functions/MF_Meltdown_CameraCull}
  n16 = Constant() {r=1}
  n17 = ComponentMask(None=CameraPositionWS()) {b=true}
  n18 = ComponentMask(None=TransformPosition(None=Constant3Vector()) {transformSourceType=TRANSFORMPOSSOURCE_Instance, transformType=TRANSFORMPOSSOURCE_World}) {b=true}
  n19 = Clamp(
   None=Divide(A=Subtract(A=@n17, B=@n18), B=ScalarParameter() {defaultValue=4, parameterName=WPOCameraFalloff}) {constB=300}
   Min=ScalarParameter() {defaultValue=-200, parameterName=WPODistance.MinOffset}
  ) {maxDefault=0}
  n20 = Constant3Vector() {constant=<rgba 0 0 1 0>}
  n21 = GetMaterialAttributes(None=SetMaterialAttributes!(
    MaterialAttributes
    Emissive Color=LinearInterpolate(A=@n11, B=@n12, Alpha=@n14)
    Opacity Mask=ShadowReplace(Default=@n15.Result, Shadow=@n16)
    World Position Offset=Multiply(A=@n19, B=@n20)
   ) {attributeSetTypes [#1, #2, #3]}) {attributeGetTypes [#1]}
  n22 = Divide(A=@n21.Emissive Color, B=EyeAdaptation().EyeAdaptation)
  StaticSwitchParameter(
   True=SetMaterialAttributes!(MaterialAttributes, Emissive Color=@n6, Opacity=@n6, Opacity Mask=@n6) {attributeSetTypes [#1, #4, #2]}
   False=SetMaterialAttributes!(MaterialAttributes=@n21.MaterialAttributes, Emissive Color=LinearInterpolate(A=@n21.Emissive Color, B=@n22)) {attributeSetTypes [#1]}
  ) {parameterName=DebugOff}
 ]
}
