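# Material function that builds a world-space linear gradient along the line from a Start point to an End point.
# It projects the world position onto that line, divides the projected distance from Start by the point-to-point distance, and applies the dot product's sign for direction.
# A UseLocalSpace static switch transforms both points from local to world space.
# Outputs are signed and bi-directional gradients, each also available clamped.
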MaterialFunction /WildEstate/TechArt/MaterialFunctions/MF_WSLinearGradient
graph {nodes [
  n1 = FunctionInput(Preview=Constant3Vector() {desc=Top / End, constant=<rgba 0 0 64 0>}) {inputName=End, sortPriority=2}
  n2 = StaticSwitch(True=TransformPosition(None=@n1) {transformType=TRANSFORMPOSSOURCE_World}, False=@n1, Value=NamedRerouteUsage())
  n3 = FunctionInput(Preview=Constant3Vector() {desc=Bottom / Line Point}) {inputName=Start}
  n4 = StaticSwitch(True=TransformPosition(None=@n3) {transformType=TRANSFORMPOSSOURCE_World}, False=@n3, Value=NamedRerouteUsage())
  n5 = Normalize(VectorInput=Subtract(A=@n2, B=@n4))
  n6 = WorldPosition() {desc=Pnt}
  n7 = DotProduct(
   A=Subtract(A=FunctionInput(Preview=@n6.XYZ) {bUsePreviewValueAsDefault=true, inputName=WorldPosition, sortPriority=3}, B=@n4) {desc=v}
   B=@n5
  )
  n8 = Multiply(A=@n5, B=@n7)
  n9 = Divide(A=Distance(A=@n4, B=Add(A=@n4, B=@n8) {desc=New Pnt Location on Line to Sample}), B=NamedRerouteUsage()) {desc=Gradient}
  n10 = Multiply(A=@n9, B=Sign(None=@n7))
  FunctionOutput(None=Saturate(None=@n10)) {outputName=Gradient Clamped, sortPriority=1}
  FunctionOutput(None=@n10) {outputName=Gradient}
  FunctionOutput(None=@n9) {outputName=Bi-Directional, sortPriority=2}
  FunctionOutput(None=Saturate(None=@n9)) {outputName=Bi-Directional Clamped, sortPriority=3}
  NamedRerouteDeclaration(None=Distance(A=@n2, B=@n4) {desc=Distance Between Points}) {name=PointDistance, nodeColor=<rgba 1 0.4219 0 1>}
  NamedRerouteDeclaration(None=FunctionInput(Preview=StaticBool()) {bUsePreviewValueAsDefault=true, inputName=UseLocalSpace, inputType=FunctionInput_StaticBool, sortPriority=4}) {name=Bool, nodeColor=<rgba 1 0 0.8672 1>}
 ]}
